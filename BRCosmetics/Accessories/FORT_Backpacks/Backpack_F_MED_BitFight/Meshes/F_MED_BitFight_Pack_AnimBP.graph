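# Post-process style Anim Blueprint for the BitFight backpack cosmetic, giving its floating cubes springy secondary motion.
# The AnimGraph nudges base_main_cube, runs six chained linear-spring, rotation-locked Anim Dynamics cube bones, then applies additive bone-space translation nodes.
# The EventGraph builds sine-wave bobbing vectors from real time and sets front-end and zipline scale variables.
# Customization load records whether the character is female.
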AnimBlueprint /BRCosmetics/Accessories/FORT_Backpacks/Backpack_F_MED_BitFight/Meshes/F_MED_BitFight_Pack_AnimBP
target_skeleton=&/F_MED_BitFight_Pack_Skeleton
graphs [
 AnimGraph {nodes [
   n1 = "Transform (Modify) Bone - Bone: base_main_cube"!(ComponentPose=Local To Component(LocalPose=Input Pose()), bAlphaBoolEnabled:bool=True, Translation="-2,0,0") -> (Pose) {
    alphaInputType=Bool
    translationMode=BMM_Additive
    translationSpace=BCS_BoneSpace
    alphaBoolBlend {blendOption=Linear}
    boneToModify {boneName=base_main_cube}
    translation {x=-2}
   }
   n2 = Anim Dynamics - Bone: dyn_main_cube!(ComponentPose=@n1, Alpha=1, ComponentLinearAccScale:Vector="0.35,0.35,0.5", ComponentLinearVelScale:Vector="0.35,0.35,0.5") -> (Pose) {
    angularDampingOverride=1
    angularSpringConstant=20000
    bLinearSpring=true
    bOverrideLinearDamping=true
    bUseGravityOverride=true
    gravityScale=0.35
    linearDampingOverride=1
    linearSpringConstant=50
    simulationSpace=RootRelative
    alphaBoolBlend {blendOption=Linear}
    boundBone {boneName=dyn_main_cube}
    componentAppliedLinearAccClamp=<1000 1000 1000>
    componentLinearAccScale=<0.35 0.35 0.5>
    componentLinearVelScale=<0.35 0.35 0.5>
    gravityOverride {z=-950}
    physicsBodyDefinitions [{
      boundBone {boneName=dyn_main_cube}
      boxExtents=<80 80 80>
      constraintSetup {angularLimitsMax=<0 0 0>, angularLimitsMin=<0 0 0>, angularTarget=<1 0 0>, linearAxesMax=<1 0.5 2.5>, linearAxesMin=<0 -0.5 -2.5>}
      localJointOffset=<1 0 0>
     }]
   }
   n3 = Anim Dynamics - Bone: dyn_cube_r_0!(ComponentPose=@n2, Alpha=1, ComponentLinearAccScale:Vector="0.35,0.35,0.55", ComponentLinearVelScale:Vector="0.35,0.35,0.55") -> (Pose) {
    angularDampingOverride=1
    angularSpringConstant=20000
    bGravityOverrideInSimSpace=true
    bLinearSpring=true
    bOverrideLinearDamping=true
    bUseGravityOverride=true
    gravityScale=0.35
    linearDampingOverride=1
    linearSpringConstant=50
    simulationSpace=RootRelative
    alphaBoolBlend {blendOption=Linear}
    boundBone {boneName=dyn_cube_r_0}
    componentAppliedLinearAccClamp=<1000 1000 1000>
    componentLinearAccScale=<0.35 0.35 0.55>
    componentLinearVelScale=<0.35 0.35 0.55>
    gravityOverride {z=-850}
    physicsBodyDefinitions [{
      boundBone {boneName=dyn_cube_r_0}
      boxExtents=<60 60 60>
      constraintSetup {angularLimitsMax=<0 0 0>, angularLimitsMin=<0 0 0>, angularTarget=<1 0 0>, linearAxesMax=<1 0.5 4>, linearAxesMin=<0 -0.5 0>}
      localJointOffset=<5 0 0>
     }]
   }
   n4 = Anim Dynamics - Bone: dyn_cube_r_1!(ComponentPose=@n3, Alpha=1, ComponentLinearAccScale:Vector="0.35,0.35,0.4", ComponentLinearVelScale:Vector="0.35,0.35,0.4") -> (Pose) {
    angularDampingOverride=1
    angularSpringConstant=20000
    bLinearSpring=true
    bOverrideLinearDamping=true
    bUseGravityOverride=true
    gravityScale=0.35
    linearDampingOverride=1
    linearSpringConstant=50
    simulationSpace=RootRelative
    alphaBoolBlend {blendOption=Linear}
    boundBone {boneName=dyn_cube_r_1}
    componentAppliedLinearAccClamp=<1000 1000 1000>
    componentLinearAccScale=<0.35 0.35 0.4>
    componentLinearVelScale=<0.35 0.35 0.4>
    gravityOverride {z=-980}
    physicsBodyDefinitions [{
      boundBone {boneName=dyn_cube_r_1}
      boxExtents=<70 70 70>
      constraintSetup {angularLimitsMax=<0 0 0>, angularLimitsMin=<0 0 0>, angularTarget=<1 0 0>, linearAxesMax=<3 1.5 5>, linearAxesMin=<0 0 -2.5>}
      localJointOffset=<10 0 0>
     }]
   }
   n5 = Anim Dynamics - Bone: dyn_cube_l_0!(ComponentPose=@n4, Alpha=1, ComponentLinearAccScale:Vector="0.35,0.35,0.4", ComponentLinearVelScale:Vector="0.35,0.35,0.4") -> (Pose) {
    angularDampingOverride=1
    angularSpringConstant=20000
    bLinearSpring=true
    bOverrideLinearDamping=true
    bUseGravityOverride=true
    gravityScale=0.35
    linearDampingOverride=1
    linearSpringConstant=50
    simulationSpace=RootRelative
    alphaBoolBlend {blendOption=Linear}
    boundBone {boneName=dyn_cube_l_0}
    componentAppliedLinearAccClamp=<1000 1000 1000>
    componentLinearAccScale=<0.35 0.35 0.4>
    componentLinearVelScale=<0.35 0.35 0.4>
    gravityOverride {z=-980}
    physicsBodyDefinitions [{
      boundBone {boneName=dyn_cube_l_0}
      boxExtents=<70 70 70>
      constraintSetup {angularLimitsMax=<0 0 0>, angularLimitsMin=<0 0 0>, angularTarget=<1 0 0>, linearAxesMax=<3 1.5 5>, linearAxesMin=<0 0 -2.5>}
      localJointOffset=<10 0 0>
     }]
   }
   n6 = Anim Dynamics - Bone: dyn_cube_l_1!(ComponentPose=@n5, Alpha=1, ComponentLinearAccScale:Vector="0.35,0.35,0.55", ComponentLinearVelScale:Vector="0.35,0.35,0.55") -> (Pose) {
    angularDampingOverride=1
    angularSpringConstant=20000
    bLinearSpring=true
    bOverrideLinearDamping=true
    bUseGravityOverride=true
    gravityScale=0.35
    linearDampingOverride=1
    linearSpringConstant=50
    simulationSpace=RootRelative
    alphaBoolBlend {blendOption=Linear}
    boundBone {boneName=dyn_cube_l_1}
    componentAppliedLinearAccClamp=<1000 1000 1000>
    componentLinearAccScale=<0.35 0.35 0.55>
    componentLinearVelScale=<0.35 0.35 0.55>
    gravityOverride {z=-930}
    physicsBodyDefinitions [{
      boundBone {boneName=dyn_cube_l_1}
      boxExtents=<60 60 60>
      constraintSetup {angularLimitsMax=<0 0 0>, angularLimitsMin=<0 0 0>, angularTarget=<1 0 0>, linearAxesMax=<1 2 6>, linearAxesMin=<0 -0.5 0>}
      localJointOffset=<5 0 0>
     }]
   }
   n7 = Anim Dynamics - Bone: dyn_cube_l_2!(ComponentPose=@n6, Alpha=1, ComponentLinearAccScale:Vector="0.35,0.35,0.45", ComponentLinearVelScale:Vector="0.35,0.35,0.45") -> (Pose) {
    angularDampingOverride=1
    angularSpringConstant=20000
    bLinearSpring=true
    bOverrideLinearDamping=true
    bUseGravityOverride=true
    gravityScale=0.35
    linearDampingOverride=1
    linearSpringConstant=50
    simulationSpace=RootRelative
    alphaBoolBlend {blendOption=Linear}
    boundBone {boneName=dyn_cube_l_2}
    componentAppliedLinearAccClamp=<1000 1000 1000>
    componentLinearAccScale=<0.35 0.35 0.45>
    componentLinearVelScale=<0.35 0.35 0.45>
    gravityOverride {z=-980}
    physicsBodyDefinitions [{
      boundBone {boneName=dyn_cube_l_2}
      boxExtents=<50 50 50>
      constraintSetup {angularLimitsMax=<0 0 0>, angularLimitsMin=<0 0 0>, angularTarget=<1 0 0>, linearAxesMax=<3 1.5 3>, linearAxesMin=<0 0 -2.5>}
      localJointOffset=<8 0 0>
     }]
   }
   n8 = "Transform (Modify) Bone - Bone: dyn_main_cube"!(ComponentPose=@n7, Alpha=1, Translation="0,0,0") -> (Pose) {translationMode=BMM_Additive, translationSpace=BCS_BoneSpace, alphaBoolBlend {blendOption=Linear}, boneToModify {boneName=dyn_main_cube}}
   n9 = "Transform (Modify) Bone - Bone: dyn_cube_r_0"!(ComponentPose=@n8, Alpha=1, Translation="0,0,0") -> (Pose) {translationMode=BMM_Additive, translationSpace=BCS_BoneSpace, alphaBoolBlend {blendOption=Linear}, boneToModify {boneName=dyn_cube_r_0}}
   n10 = "Transform (Modify) Bone - Bone: dyn_cube_r_1"!(ComponentPose=@n9, Alpha=1, Translation="0,0,0") -> (Pose) {translationMode=BMM_Additive, translationSpace=BCS_BoneSpace, alphaBoolBlend {blendOption=Linear}, boneToModify {boneName=dyn_cube_r_1}}
   n11 = "Transform (Modify) Bone - Bone: dyn_cube_l_0"!(ComponentPose=@n10, Alpha=1, Translation="0,0,0") -> (Pose) {translationMode=BMM_Additive, translationSpace=BCS_BoneSpace, alphaBoolBlend {blendOption=Linear}, boneToModify {boneName=dyn_cube_l_0}}
   n12 = "Transform (Modify) Bone - Bone: dyn_cube_l_1"!(ComponentPose=@n11, Alpha=1, Translation="0,0,0") -> (Pose) {translationMode=BMM_Additive, translationSpace=BCS_BoneSpace, alphaBoolBlend {blendOption=Linear}, boneToModify {boneName=dyn_cube_l_1}}
   n13 = "Transform (Modify) Bone - Bone: dyn_cube_l_2"!(ComponentPose=@n12, Alpha=1, Translation="0,0,0") -> (Pose) {translationMode=BMM_Additive, translationSpace=BCS_BoneSpace, alphaBoolBlend {blendOption=Linear}, boneToModify {boneName=dyn_cube_l_2}}
   Output Pose(Result=Component To Local(ComponentPose=@n13)) {layerGroup=None}
  ]}
 EventGraph {nodes [
   Event BlueprintUpdateAnimation()
   > n1 = Sequence()
   n2 = Event BlueprintProcessAnimInputEvent()
   n3 = IsNamed(self=@n2.AnimInputEvent, InEventName=aie.Player.Zipline.Entered)
   n4 = OR Boolean(A=Get bIsMoving2D(), B=Get bIsTacticalSprinting())
   n5 = Get bIsJumping()
   n6 = Branch(execute=@n1.then_1, Condition=OR Boolean(A=@n3, B=AND Boolean!(A=@n3, B=@n5, C:bool=@n4) -> (ReturnValue)))
   > Set animDyn on zipline(animDyn on zipline:Vector="0, 0, 0") -> (Output_Get:Vector)
   Branch(execute=@n6.else, Condition=IsNamed(self=@n2.AnimInputEvent, InEventName=aie.Player.Zipline.Exited))
   > Set animDyn on zipline(animDyn on zipline="0.35,0.35,0.35")
   n7 = GetRealTimeSeconds()
   n8 = "Sin (Radians)"(A=float * float(A=float * float(A=1.3, B=1.3), B=@n7))
   n9 = "Sin (Radians)"(A=float * float(A=float * float(A=1.1, B=1.1), B=@n7))
   n10 = "Sin (Radians)"(A=float * float(A=float * float(A=0.9, B=1.1), B=@n7))
   Set main cube float(
    execute=@n1.then_0
    main cube float:Vector=MakeVector(X=float * float(A=@n10, B=0.3), Y=float * float(A=@n9, B=0.15), Z=float * float(A=@n8, B=1.2))
   ) -> (Output_Get:Vector)
   n11 = "Sin (Radians)"(A=float * float(A=float * float(A=0.54, B=1.08), B=@n7))
   n12 = "Sin (Radians)"(A=float * float(A=float * float(A=0.9, B=0.66), B=@n7))
   n13 = "Sin (Radians)"(A=float * float(A=float * float(A=0.9, B=0.78), B=@n7))
   > Set cube_0_r float(cube_0_r float:Vector=MakeVector(X=float * float(A=@n13, B=1.2), Y=float * float(A=@n12, B=0.15), Z=float * float(A=@n11, B=0.5))) -> (Output_Get:Vector)
   n14 = "Sin (Radians)"(A=float * float(A=float * float(A=1.02, B=0.66), B=@n7))
   n15 = "Sin (Radians)"(A=float * float(A=float * float(A=0.78, B=0.9), B=@n7))
   n16 = "Sin (Radians)"(A=float * float(A=float * float(A=0.3, B=0.66), B=@n7))
   > Set cube_1_r float(cube_1_r float:Vector=MakeVector(X=float * float(A=@n15, B=0.8), Y=float * float(A=@n14, B=0.5), Z=float * float(A=@n16, B=0.9))) -> (Output_Get:Vector)
   n17 = "Sin (Radians)"(A=float * float(A=float * float(A=0.54, B=0.66), B=@n7))
   n18 = "Sin (Radians)"(A=float * float(A=float * float(A=0.66, B=0.66), B=@n7))
   n19 = "Sin (Radians)"(A=float * float(A=float * float(A=0.78, B=0.78), B=@n7))
   > Set cube_2_l float(cube_2_l float:Vector=MakeVector(X=float * float(A=@n19, B=0.5), Y=float * float(A=@n18, B=0.15), Z=float * float(A=@n17, B=0.8))) -> (Output_Get:Vector)
   > n20 = Branch(Condition=Get bIsInFrontEnd())
   > Set is Front End(is Front End:double=0.5) -> (Output_Get:double)
   Set is Front End(execute=@n20.else, is Front End=1)
   n21 = Event OnCharacterCustomizationLoaded()
   n22 = Get Array Item with EFort Custom Part Type(Array:Array of Custom Character Part Object References=@n21.CustomParts) -> (Output:Custom Character Part Object Reference)
   > Set gender female(gender female:bool="Equal (Enum)"(
     A:EFortCustomGender Enum=Get GenderPermitted!(self:Custom Character Part Object Reference=@n22) -> (GenderPermitted:EFortCustomGender Enum=Invalid)
     B:EFortCustomGender Enum=Female
    )) -> (Output_Get:bool)
   TryGetPawnOwner()
  ]}
]
variables [
 animDyn on zipline {type=Vector}
 cube_0_r float {type=Vector}
 cube_1_r float {type=Vector}
 cube_2_l float {type=Vector}
 gender female {type=boolean}
 is Front End {type=number, default=1}
 main cube float {type=Vector}
]
